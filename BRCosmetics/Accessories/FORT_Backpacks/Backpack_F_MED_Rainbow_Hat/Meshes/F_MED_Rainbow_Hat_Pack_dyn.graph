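# Physics asset for the Rainbow Hat backpack cosmetic, defining collision and joint limits for its dangling dyn_ bones.
# The pack body is the hub for the bag cover, wire, three claw bodies and two large two-segment ear chains.
# The left and right claws each carry a head with two small ear chains, plus two arms and two legs.
# Every joint keeps twist tight and allows wider, often asymmetric swing.
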PhysicsAsset /BRCosmetics/Accessories/FORT_Backpacks/Backpack_F_MED_Rainbow_Hat/Meshes/F_MED_Rainbow_Hat_Pack_dyn
bodies {
 dyn_arms_01_l [{shapeType=Capsule, length=0.75, radius=1.25, center=<1.5916 0 0.0044>, rotation=<rot 85 0 0>}]
 dyn_arms_01_r [{shapeType=Capsule, length=1, radius=1, center=<1.723 0 -0.0942>, rotation=<rot 85 0 0>}]
 dyn_arms_02_l [{shapeType=Capsule, length=0.75, radius=1.25, center=<1.6924 0 -0.0534>, rotation=<rot 90 -0.1187 -0.1187>}]
 dyn_arms_02_r [{shapeType=Capsule, length=1, radius=1, center=<1.2761 0 0.0395>, rotation=<rot 80 0 0>}]
 dyn_bag_cover [{shapeType=Capsule, length=9.5, radius=1.2, center=<10 0 0>, rotation=<rot 0 0 90>}]
 dyn_claw_l [{shapeType=Capsule, length=3.75, radius=3.65, center=<1.93 -0.1058 3.1296>, rotation=<rot 0 0 90>}]
 dyn_claw_lw [{shapeType=Capsule, length=5.7032, radius=4.2103, center=<4 0 0>, rotation=<rot 0 0 90>}]
 dyn_claw_r [{shapeType=Capsule, length=3.75, radius=3.65, center=<2.1696 -0.2193 3.5>, rotation=<rot 0 0 90>}]
 dyn_ear_01_l [{shapeType=Box, extentX=8, extentY=5, extentZ=1.5, center=<4 0 0>}]
 dyn_ear_01_r [{shapeType=Box, extentX=9, extentY=5, extentZ=1.5, center=<4.5 0 0>}]
 dyn_ear_02_l [{shapeType=Box, extentX=8, extentY=5, extentZ=1.5, center=<4 0 0>}]
 dyn_ear_02_r [{shapeType=Box, extentX=6, extentY=5, extentZ=1.5, center=<3 0 0>}]
 dyn_ear_a_01_l [{shapeType=Box, extentX=3, extentY=3, extentZ=1, center=<1.5 0 0>, rotation=<rot 0 0.0005 0>}]
 dyn_ear_a_01_r [{shapeType=Box, extentX=3, extentY=3, extentZ=1, center=<1.5 0 0>, rotation=<rot 0 0.0005 0>}]
 dyn_ear_a_02_l [{shapeType=Box, extentX=3, extentY=3, extentZ=1, center=<1.5 0 0>, rotation=<rot 0 0.0005 0>}]
 dyn_ear_a_02_r [{shapeType=Box, extentX=4, extentY=3, extentZ=1, center=<2 0 0>, rotation=<rot 0 0.0005 0>}]
 dyn_ear_b_01_l [{shapeType=Box, extentX=3, extentY=3, extentZ=1, center=<1.5 0 0>, rotation=<rot 0 0.0004 0>}]
 dyn_ear_b_01_r [{shapeType=Box, extentX=3, extentY=3, extentZ=1, center=<1.5 0 0>, rotation=<rot 0 0.0004 0>}]
 dyn_ear_b_02_l [{shapeType=Box, extentX=3, extentY=3, extentZ=1, center=<1.5 0 0>, rotation=<rot 0 0.0004 0>}]
 dyn_ear_b_02_r [{shapeType=Box, extentX=4, extentY=3, extentZ=1, center=<2 0 0>, rotation=<rot 0 0.0005 0>}]
 dyn_head_l [{shapeType=Capsule, length=1.4, radius=3.5, center=<3.8 0 0>, rotation=<rot 0.0001 0 90>}]
 dyn_head_r [{shapeType=Capsule, length=1.4, radius=3.5, center=<3.8 0 0>, rotation=<rot 0.0001 0 90>}]
 dyn_legs_01_l [{shapeType=Capsule, length=2.5, radius=1.5, shapeName=dyn_legs_01_l_capsule, center=<1 0 1>}]
 dyn_legs_01_r [{shapeType=Capsule, length=1.5, radius=1.5, shapeName=dyn_legs_01_r_capsule, center=<1 0 0>, rotation=<rot -90 0 0>}]
 dyn_legs_02_l [{shapeType=Capsule, length=2.5, radius=1.7, shapeName=dyn_legs_02_l_capsule, center=<1 0 1>}]
 dyn_legs_02_r [{shapeType=Capsule, length=1.5, radius=1.4, shapeName=dyn_legs_02_r_capsule, center=<1 0 0>, rotation=<rot -90 0 0>}]
 dyn_wire [{shapeType=Capsule, length=2.0282, radius=2.7069, center=<9.0161 0 1.0364>, rotation=<rot -10.0004 179.9999 -180>}]
 pack [
  {shapeType=Capsule, length=9, radius=5.5, center=<0 -4.8782 2.3877>, rotation=<rot -90 0.0132 -0.013>}
  {shapeType=Box, extentX=15.5, extentY=10, extentZ=21.5, center=<0 -4.8917 -6.6905>}
  {shapeType=Box, extentX=15.5, extentY=8.5, extentZ=21, center=<-0.0001 -7.5166 -7.2178>, rotation=<rot 0 0 15.0038>}
 ]
}
constraints [
 {bone1Name=dyn_claw_lw, bone2Name=pack, swing1LimitDegrees=20, swing2LimitDegrees=5, twistLimitDegrees=5}
 {bone1Name=dyn_wire, bone2Name=pack, swing1LimitDegrees=5, swing2LimitDegrees=30, twistLimitDegrees=5}
 {bone1Name=dyn_claw_r, bone2Name=pack, swing1LimitDegrees=15, swing2LimitDegrees=15, twistLimitDegrees=5}
 {bone1Name=dyn_head_r, bone2Name=dyn_claw_r, swing1LimitDegrees=15, swing2LimitDegrees=15, twistLimitDegrees=5}
 {bone1Name=dyn_ear_a_01_r, bone2Name=dyn_head_r, swing1LimitDegrees=10, swing2LimitDegrees=20, twistLimitDegrees=5}
 {bone1Name=dyn_ear_a_02_r, bone2Name=dyn_ear_a_01_r, swing1LimitDegrees=10, swing2LimitDegrees=20, twistLimitDegrees=5}
 {bone1Name=dyn_ear_b_01_r, bone2Name=dyn_head_r, swing1LimitDegrees=10, swing2LimitDegrees=20, twistLimitDegrees=5}
 {bone1Name=dyn_ear_b_02_r, bone2Name=dyn_ear_b_01_r, swing1LimitDegrees=10, swing2LimitDegrees=20, twistLimitDegrees=5}
 {bone1Name=dyn_arms_01_r, bone2Name=dyn_claw_r, swing1LimitDegrees=5, swing2LimitDegrees=35, twistLimitDegrees=2}
 {bone1Name=dyn_arms_02_r, bone2Name=dyn_claw_r, swing1LimitDegrees=5, swing2LimitDegrees=35, twistLimitDegrees=2}
 {bone1Name=dyn_claw_l, bone2Name=pack, swing1LimitDegrees=15, swing2LimitDegrees=15, twistLimitDegrees=5}
 {bone1Name=dyn_head_l, bone2Name=dyn_claw_l, swing1LimitDegrees=15, swing2LimitDegrees=15, twistLimitDegrees=5}
 {bone1Name=dyn_ear_a_01_l, bone2Name=dyn_head_l, swing1LimitDegrees=10, swing2LimitDegrees=20, twistLimitDegrees=5}
 {bone1Name=dyn_ear_a_02_l, bone2Name=dyn_ear_a_01_l, swing1LimitDegrees=10, swing2LimitDegrees=20, twistLimitDegrees=5}
 {bone1Name=dyn_ear_b_01_l, bone2Name=dyn_head_l, swing1LimitDegrees=10, swing2LimitDegrees=20, twistLimitDegrees=5}
 {bone1Name=dyn_ear_b_02_l, bone2Name=dyn_ear_b_01_l, swing1LimitDegrees=10, swing2LimitDegrees=20, twistLimitDegrees=5}
 {bone1Name=dyn_arms_01_l, bone2Name=dyn_claw_l, swing1LimitDegrees=5, swing2LimitDegrees=35, twistLimitDegrees=2}
 {bone1Name=dyn_arms_02_l, bone2Name=dyn_claw_l, swing1LimitDegrees=5, swing2LimitDegrees=35, twistLimitDegrees=2}
 {bone1Name=dyn_bag_cover, bone2Name=pack, swing1LimitDegrees=25, swing2LimitDegrees=75, twistLimitDegrees=5}
 {bone1Name=dyn_ear_01_l, bone2Name=pack, swing1LimitDegrees=10, swing2LimitDegrees=35, twistLimitDegrees=5}
 {bone1Name=dyn_ear_02_l, bone2Name=dyn_ear_01_l, swing1LimitDegrees=10, swing2LimitDegrees=35, twistLimitDegrees=5}
 {bone1Name=dyn_ear_01_r, bone2Name=pack, swing1LimitDegrees=10, swing2LimitDegrees=35, twistLimitDegrees=5}
 {bone1Name=dyn_ear_02_r, bone2Name=dyn_ear_01_r, swing1LimitDegrees=10, swing2LimitDegrees=35, twistLimitDegrees=5}
 {bone1Name=dyn_legs_02_l, bone2Name=dyn_claw_l, swing1LimitDegrees=35, swing2LimitDegrees=35, twistLimitDegrees=5}
 {bone1Name=dyn_legs_01_l, bone2Name=dyn_claw_l, swing1LimitDegrees=35, swing2LimitDegrees=35, twistLimitDegrees=5}
 {bone1Name=dyn_legs_01_r, bone2Name=dyn_claw_r, swing1LimitDegrees=30, swing2LimitDegrees=30, twistLimitDegrees=3}
 {bone1Name=dyn_legs_02_r, bone2Name=dyn_claw_r, swing1LimitDegrees=35, swing2LimitDegrees=35, twistLimitDegrees=5}
]
